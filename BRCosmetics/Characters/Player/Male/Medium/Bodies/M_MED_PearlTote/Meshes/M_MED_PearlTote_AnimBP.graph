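# Animation Blueprint for the PearlTote male body cosmetic, layering secondary collar and zipper motion onto the pose copied from its parent mesh.
# A Control Rig receives a smoothed targeting value, then feeds a zipper-base constraint and Anim Dynamics on the zipper.
# An additive bone rotation bends the back collar during swim sprinting.
# On customization load, back bling detection toggles a material parameter that hides an element.
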AnimBlueprint /BRCosmetics/Characters/Player/Male/Medium/Bodies/M_MED_PearlTote/Meshes/M_MED_PearlTote_AnimBP
target_skeleton=&/M_MED_PearlTote_Skeleton
graphs [
 AnimGraph {nodes [
   Output Pose(Result=Use cached pose 'Cache_output'()) {layerGroup=None}
   n1 = "Constraint - base_zipper (P)"!(ComponentPose=Local To Component(LocalPose=Use cached pose 'Cache_Input'()), Alpha=1, ConstraintWeights_0=1, ConstraintWeights_1:float=0) -> (Pose) {
    alphaBoolBlend {blendOption=Linear}
    boneToModify {boneName=base_zipper}
    constraintSetup [{targetBone {boneName=dyn_collar_fr_01}}]
    constraintWeights [1,0]
   }
   n2 = Anim Dynamics - Bone: dyn_zipper(ComponentPose=@n1, Alpha=1) {
    angularDampingOverride=1
    angularSpringConstant=20000
    bOverrideAngularDamping=true
    bUseGravityOverride=true
    gravityScale=0.35
    simulationSpace=RootRelative
    alphaBoolBlend {blendOption=Linear}
    boundBone {boneName=dyn_zipper}
    componentAppliedLinearAccClamp=<10000 10000 10000>
    componentLinearAccScale=<1 1 1>
    componentLinearVelScale=<1 1 1>
    gravityOverride {z=-980}
    physicsBodyDefinitions [{
      boundBone {boneName=dyn_zipper}
      boxExtents=<10 10 8>
      constraintSetup {angularLimitsMax=<5 5 60>, angularLimitsMin=<-5 -5 -20>, angularTarget=<1 0 0>}
      localJointOffset=<3 0 0>
     }]
   }
   n3 = "Transform (Modify) Bone - Bone: dyn_collar_bk_01"!(ComponentPose=@n2, bAlphaBoolEnabled:bool=Get bIsSwimSprinting() -> (bIsSwimSprinting:bool), Rotation="0,-18,0") -> (Pose) {
    alphaInputType=Bool
    rotationMode=BMM_Additive
    rotationSpace=BCS_ParentBoneSpace
    alphaBoolBlend {blendInTime=0.2, blendOption=CubicInOut, blendOutTime=0.2}
    boneToModify {boneName=dyn_collar_bk_01}
    rotation {yaw=-18}
   }
   Save cached pose 'Cache_output'(Pose=Component To Local(ComponentPose=@n3)) {cacheName=Cache_output}
   n4 = Control Rig!(Source=Copy Pose From Mesh() {bUseAttachedParent=true}, Alpha, isTargetingF:double=Get isTargeting float() -> (isTargeting float:double)) -> (Pose) {
    alphaBoolBlend {blendOption=Linear}
    controlRigAssetReference {blueprintRigClass=&/M_MED_PearlTote_CtrlRig.M_MED_PearlTote_CtrlRig_C}
   }
   Save cached pose 'Cache_Input'(Pose=@n4) {cacheName=Cache_Input}
  ]}
 EventGraph {nodes [
   n1 = Event BlueprintUpdateAnimation()
   n2 = Get bIsTargeting()
   > Set isTargeting float(isTargeting float:double=FInterpTo(Current=Get isTargeting float(), Target="To Float (Boolean)"(InBool=@n2), DeltaTime=@n1.DeltaTimeX, InterpSpeed=15)) -> (Output_Get:double)
   n3 = Event OnCharacterCustomizationLoaded()
   > Set SkelMesh(SkelMesh=@n3.CustomBaseMesh)
   n4 = CheckForBackbling()
   > Set isWearingBackbling(isWearingBackbling=OR Boolean!(A=@n4, B=@n4.bIsBackblingACape, C:bool=@n4.bIsBackblingLong) -> (ReturnValue))
   > n5 = Get Dynamic Material Instance(SkeletalMeshComponent=GetPartSkeletalMeshComponent())
   > n6 = Branch(Condition=Get isWearingBackbling())
   > SetScalarParameterValue(self=@n5, ParameterName=Hide Element 01, Value=1)
   SetScalarParameterValue(execute=@n6.else, self=@n5, ParameterName=Hide Element 01, Value=0)
  ]}
]
variables [
 isTargeting float {type=number}
 isWearingBackbling {type=boolean}
 SkelMesh {type=SkeletalMeshComponent}
]
