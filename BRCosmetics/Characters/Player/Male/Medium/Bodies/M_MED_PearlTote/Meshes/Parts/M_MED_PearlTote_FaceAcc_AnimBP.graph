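# Post-processes the PearlTote male cosmetic's face accessory, adding spring-driven ear motion and damped head and neck rotation.
# The AnimGraph copies the attached parent's pose, chains bone-relative Anim Dynamics on both two-bone ears, then modifies head and, during emotes, neck_01.
# The EventGraph derives additive counter-rotations from ref-pose deltas, scaled per axis with yaw-dependent yaw scaling.
# It also computes a gravity override via a deprecated function, though the dynamics nodes use zero literals.
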AnimBlueprint /BRCosmetics/Characters/Player/Male/Medium/Bodies/M_MED_PearlTote/Meshes/Parts/M_MED_PearlTote_FaceAcc_AnimBP
target_skeleton=&/M_MED_PearlTote_FaceAcc_Skeleton
graphs [
 AnimGraph {nodes [
   Output Pose(Result=Use cached pose 'Input animDyn Cache'()) {layerGroup=None}
   n1 = Anim Dynamics - Bone: dyn_ear_r_01!(ComponentPose=Local To Component(LocalPose=Copy Pose From Mesh() {bUseAttachedParent=true}), Alpha=1, GravityOverride:Vector="0,0,0") -> (Pose) {
    angularDampingOverride=1
    angularSpringConstant=4000
    bAngularSpring=true
    bOverrideAngularDamping=true
    bUseGravityOverride=true
    gravityScale=0.35
    simulationSpace=BoneRelative
    alphaBoolBlend {blendOption=Linear}
    boundBone {boneName=dyn_ear_r_01}
    componentAppliedLinearAccClamp=<10000 10000 10000>
    componentLinearAccScale=<1 1 1>
    componentLinearVelScale=<1 1 1>
    physicsBodyDefinitions [{
      boundBone {boneName=dyn_ear_r_01}
      boxExtents=<18 18 18>
      constraintSetup {angularLimitsMax=<0 3.5 10>, angularLimitsMin=<0 -3.5 -10>, angularTarget=<1 0 0>}
      localJointOffset=<4 0 0>
     }]
    relativeSpaceBone {boneName=spine_03}
   }
   n2 = Anim Dynamics - Bone: dyn_ear_r_02!(ComponentPose=@n1, Alpha=1, GravityOverride:Vector="0,0,0") -> (Pose) {
    angularDampingOverride=1
    angularSpringConstant=4000
    bAngularSpring=true
    bOverrideAngularDamping=true
    bUseGravityOverride=true
    gravityScale=0.35
    simulationSpace=BoneRelative
    alphaBoolBlend {blendOption=Linear}
    boundBone {boneName=dyn_ear_r_02}
    componentAppliedLinearAccClamp=<10000 10000 10000>
    componentLinearAccScale=<1 1 1>
    componentLinearVelScale=<1 1 1>
    physicsBodyDefinitions [{
      boundBone {boneName=dyn_ear_r_02}
      boxExtents=<18 18 18>
      constraintSetup {angularLimitsMax=<0 0 5>, angularLimitsMin=<0 0 -5>, angularTarget=<1 0 0>}
      localJointOffset=<4 0 0>
     }]
    relativeSpaceBone {boneName=spine_03}
   }
   n3 = Anim Dynamics - Bone: dyn_ear_l_01!(ComponentPose=@n2, Alpha=1, GravityOverride:Vector="0,0,0") -> (Pose) {
    angularDampingOverride=1
    angularSpringConstant=4000
    bAngularSpring=true
    bOverrideAngularDamping=true
    bUseGravityOverride=true
    gravityScale=0.35
    simulationSpace=BoneRelative
    alphaBoolBlend {blendOption=Linear}
    boundBone {boneName=dyn_ear_l_01}
    componentAppliedLinearAccClamp=<10000 10000 10000>
    componentLinearAccScale=<1 1 1>
    componentLinearVelScale=<1 1 1>
    physicsBodyDefinitions [{
      boundBone {boneName=dyn_ear_l_01}
      boxExtents=<18 18 18>
      constraintSetup {angularLimitsMax=<0 3.5 10>, angularLimitsMin=<0 -3.5 -10>, angularTarget=<1 0 0>}
      localJointOffset=<4 0 0>
     }]
    relativeSpaceBone {boneName=spine_03}
   }
   n4 = Anim Dynamics - Bone: dyn_ear_l_02!(ComponentPose=@n3, Alpha=1, GravityOverride:Vector="0,0,0") -> (Pose) {
    angularDampingOverride=1
    angularSpringConstant=4000
    bAngularSpring=true
    bOverrideAngularDamping=true
    bUseGravityOverride=true
    gravityScale=0.35
    simulationSpace=BoneRelative
    alphaBoolBlend {blendOption=Linear}
    boundBone {boneName=dyn_ear_l_02}
    componentAppliedLinearAccClamp=<10000 10000 10000>
    componentLinearAccScale=<1 1 1>
    componentLinearVelScale=<1 1 1>
    physicsBodyDefinitions [{
      boundBone {boneName=dyn_ear_l_02}
      boxExtents=<18 18 18>
      constraintSetup {angularLimitsMax=<0 0 5>, angularLimitsMin=<0 0 -5>, angularTarget=<1 0 0>}
      localJointOffset=<4 0 0>
     }]
    relativeSpaceBone {boneName=spine_03}
   }
   n5 = "Transform (Modify) Bone - Bone: head"!(ComponentPose=@n4, Alpha=1, Rotation=Get HeadRot() -> (HeadRot:Rotator)) -> (Pose) {rotationMode=BMM_Additive, rotationSpace=BCS_BoneSpace, alphaBoolBlend {blendOption=Linear}, boneToModify {boneName=head}}
   n6 = Get bIsPlayingEmote()
   n7 = "Transform (Modify) Bone - Bone: neck_01"!(ComponentPose=@n5, bAlphaBoolEnabled:bool=@n6, Translation="0,0,0", Rotation=Get NeckRot() -> (NeckRot:Rotator)) -> (Pose) {
    alphaInputType=Bool
    rotationMode=BMM_Additive
    rotationSpace=BCS_BoneSpace
    translationMode=BMM_Replace
    translationSpace=BCS_BoneSpace
    alphaBoolBlend {blendInTime=0.1, blendOption=Linear, blendOutTime=0.1}
    boneToModify {boneName=neck_01}
   }
   n8 = "Transform (Modify) Bone - Bone: neck_01"!(ComponentPose=@n7, bAlphaBoolEnabled=@n6, Translation="0,0,0") -> (Pose) {
    alphaInputType=Bool
    rotationSpace=BCS_BoneSpace
    translationMode=BMM_Replace
    translationSpace=BCS_BoneSpace
    alphaBoolBlend {blendInTime=0.1, blendOption=Linear, blendOutTime=0.1}
    boneToModify {boneName=neck_01}
   }
   Save cached pose 'Input animDyn Cache'(Pose=Component To Local(ComponentPose=@n8)) {cacheName=Input animDyn Cache}
  ]}
 EventGraph {nodes [
   Event BlueprintUpdateAnimation()
   > n1 = Sequence!(execute) -> (then_0, then_1, then_2:Exec)
   > n2 = Gravity Override State Function DEPRECATED(CharacterPartAnimInstance=Self-Reference(), GameTimeInSeconds=GetGameTimeInSeconds(), GravityOverrideParams=Get Gravity Override Params())
   > Set Gravity Override(Gravity Override=@n2)
   Branch(execute=@n1.then_2, Condition=Get bIsPlayingEmote())
   n3 = BreakRotator(InRot=GetDeltaTransformFromRefPose!(self=GetOwnerSkeletalMeshComponent(), BoneName=neck_01, BaseName=spine_05) -> (ReturnValue_Location:Vector, ReturnValue_Rotation:Rotator, ReturnValue_Scale:Vector).ReturnValue_Rotation)
   n4 = Lerp(
    A=Get Z Percentage Min() -> (Z Percentage Min:double)
    B=Get Z Percentage Max() -> (Z Percentage Max:double)
    Alpha=Remap Float(Input=@n3.Yaw, OldRange=(X=10,Y=-28), NewRange=(X=0,Y=1), Clamp=true) -> (Output=~)
   )
   n5 = Get X Percentage() -> (X Percentage:double)
   n6 = Get Y Percentage() -> (Y Percentage:double)
   > Set NeckRot(NeckRot:Rotator=MakeRotator(
     Roll=float * float!(A=@n3, B=-1, C:double=@n5) -> (ReturnValue) {is=title K2Node_PromotableOperator}
     Pitch=float * float(A=@n3.Pitch, B=-1, C=@n6)
     Yaw=float * float(A=@n3.Yaw, B=-1, C=@n4)
    )) -> (Output_Get:Rotator)
   n7 = BreakRotator(InRot=GetDeltaTransformFromRefPose(self=GetOwnerSkeletalMeshComponent(), BoneName=head, BaseName=neck_02).ReturnValue_Rotation)
   n8 = Get X Percentage()
   n9 = Get Y Percentage()
   n10 = Lerp(
    A=Get Z Percentage Min()
    B=Get Z Percentage Max()
    Alpha=Remap Float(Input=@n7.Yaw, OldRange=(X=18,Y=-28), NewRange=(X=0,Y=1), Clamp=true) -> (Output=~)
   )
   Set HeadRot(
    execute=@n1.then_1
    HeadRot:Rotator=MakeRotator(Roll=float * float(A=@n7, B=-1, C=@n8), Pitch=float * float(A=@n7.Pitch, B=-1, C=@n9), Yaw=float * float(A=@n7.Yaw, B=-1, C=@n10))
   ) -> (Output_Get:Rotator)
  ]}
]
variables [
 Gravity Override {type=Vector}
 Gravity Override Params {type=GravityOverrideParamsStruct}
 HeadRot {type=Rotator}
 NeckRot {type=Rotator}
 X Percentage {type=number, default=0.2}
 Y Percentage {type=number, default=0.4}
 Z Percentage Max {type=number, default=0.4}
 Z Percentage Min {type=number, default=1}
]
